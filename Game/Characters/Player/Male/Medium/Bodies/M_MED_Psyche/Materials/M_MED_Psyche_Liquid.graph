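# Skinned character material for the Psyche body that adds an animated, glossy liquid look to masked regions.
# Panning world-aligned projections on pre-skinned position distort a twisted normal, driving color variation and camera-facing emissive highlights.
# A mask blends this into CharacterSkinFX output between shared PreFX and PostFX functions.
# Sine-based local and world noise wobbles vertices via WPO, gated by mask and vertex green, with camera distance scaling the local part.
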Material /Game/Characters/Player/Male/Medium/Bodies/M_MED_Psyche/Materials/M_MED_Psyche_Liquid
graph {
 properties {
  bUseMaterialAttributes=true
  bUsedWithMeshDeformer=true
  bUsedWithMorphTargets=true
  bUsedWithSkeletalMesh=true
  pixelDepthOffsetMode=PDOM_Legacy
  refractionMethod=RM_IndexOfRefraction
  usedShadingModels=Default Lit
 }
 nodes [
  n1 = ScalarParameter() {defaultValue=7.1525, parameterName=Fresnel}
  n2 = TextureObject() {samplerType=SAMPLERTYPE_Normal, texture=/Game/Characters/Player/Male/Medium/Bodies/M_MED_Mastermind/Textures/T_M_MED_Mastermind_Twisted_N}
  n3 = AppendVector(A=ScalarParameter() {defaultValue=84.6447, parameterName=Size_Normal}, B=ScalarParameter() {defaultValue=82.35, parameterName=Size_Normal2})
  n4 = ScalarParameter() {defaultValue=33.2833, parameterName=Size_Normal3}
  n5 = Constant3Vector() {constant=<rgba 0 0 1 0>}
  n6 = VertexInterpolator(VS=PreSkinnedNormal())
  n7 = Normalize(VectorInput=Transform(None=@n6.PS) {transformSourceType=TRANSFORMSOURCE_Local, transformType=TRANSFORM_Tangent})
  n8 = VertexInterpolator(VS=PreSkinnedPosition())
  n9 = TextureObject() {samplerType=SAMPLERTYPE_Masks, texture=/Game/Characters/Player/Male/Medium/Bodies/M_MED_Mastermind/Textures/T_M_MED_Mastermind_Twisted_M}
  n10 = ComponentMask(None=VectorParameter() {parameterName=Pan_Mask_YZ, defaultValue=<rgba -0.0857 0.1333 0 1>}.RGB) {g=true, r=true}
  n11 = ScalarParameter() {defaultValue=0.4286, parameterName=PanSpeedMask}
  n12 = ComponentMask(None=VectorParameter() {parameterName=Pan_Mask_XZ, defaultValue=<rgba 0.0762 0.1333 0 1>}.RGB) {g=true, r=true}
  n13 = ComponentMask(None=VectorParameter() {parameterName=Pan_Mask_XY, defaultValue=<rgba 0.0286 -0.05 0 1>}.RGB) {g=true, r=true}
  n14 = MaterialFunctionCall!(
   Export Float 4
   TextureObject=@n9
   TextureSize=ScalarParameter() {defaultValue=33.9662, parameterName=Size_DistortionMask}
   WorldPosition=@n8.PS
   World Space Normal=@n6.PS
   ProjectionTransitionContrast
   Pan_YZ=Multiply(A=@n10, B=@n11)
   Pan_XZ=Multiply(A=@n12, B=@n11)
   Pan_XY=Multiply(A=@n13, B=@n11)
  ) {materialFunction=/Game/Weapons/WeaponSkins/Wraps/MaterialFunctions/WorldAlignedTexture_ComplexPan}
  n15 = ComponentMask(None=@n14.XYZ Texture) {b=true}
  n16 = ScalarParameter() {defaultValue=0.5524, parameterName=PanSpeedNormal}
  n17 = MaterialFunctionCall(
   Normal=MaterialFunctionCall!(
    TextureObject=@n2
    TextureSize=AppendVector(A=@n3, B=@n4)
    Normal=StaticSwitchParameter(True=@n5, False=@n7) {parameterName=UseTwistedCorrectNormals}
    WorldSpace
    ProjectionTransitionContrast=ScalarParameter() {parameterName=ProjectionContrast}
    WorldPosition=Add(A=@n8.PS, B=@n15) {constB=0}
    Use High Quality Normals=StaticBoolParameter() {defaultValue=true}
    Pan_YZ=Multiply(A=@n10, B=@n16)
    Pan_XZ=Multiply(A=@n12, B=@n16)
    Pan_XY=Multiply(A=@n13, B=@n16)
   ) {materialFunction=/Game/Weapons/WeaponSkins/Wraps/MaterialFunctions/WorldAlignedNormal_ComplexPan}.XYZ Texture
   Flatness=ScalarParameter() {defaultValue=0.2476, parameterName=NormalFlatten}
  ) {materialFunction=&FlattenNormal}
  n18 = Transform(None=@n17.Result)
  n19 = DotProduct(A=CameraVectorWS(), B=Normalize(VectorInput=@n18))
  n20 = Multiply(
   A=ScalarParameter() {defaultValue=0.2762, parameterName=FresnelBirghtness2}
   B=Power(Base=@n19, Exp=ScalarParameter() {defaultValue=442.9701, parameterName=Fresnel2})
  )
  Add(
   A=Multiply(A=ScalarParameter() {defaultValue=0.0667, parameterName=FresnelBirghtness}, B=Fresnel(ExponentIn=@n1, Normal=@n18) {baseReflectFraction=0})
   B=@n20
  )
  n21 = MaterialFunctionCall() {materialFunction=&BaseFN_CharacterFunction_PreFX}
  n22 = MaterialFunctionCall(
   PreskinnedPosition=@n21.Pre skinned Local Position
   Gmap=@n21.Pre skinned Local Position
   MaterialAttributes=@n21.Pre skinned Local Position
   AO=@n21.Pre skinned Local Position
  ) {materialFunction=&CharacterSkinFX}
  n23 = GetMaterialAttributes(None=@n22.Result) {attributeGetTypes [#1, #2, #3]}
  n24 = NamedRerouteUsage()
  LinearInterpolate(A=@n23.Normal, B=@n24)
  n25 = ScalarParameter() {defaultValue=6, parameterName=Distortion}
  Multiply(A=AppendVector(A=AppendVector(A=@n25, B=@n25), B=@n25), B=ComponentMask(None=@n14.XYZ Texture) {g=true})
  Multiply(B=@n19) {constA=-1}
  n26 = PreSkinnedPosition()
  n27 = ScalarParameter() {defaultValue=1, group=Slurp, parameterName=Slurp_WPO_LocalNoise_Size}
  n28 = ScalarParameter() {defaultValue=0.02, group=Slurp, parameterName=Slurp_WPO_LocalNoise_Size2}
  n29 = MaterialFunctionCall(Float3=Multiply(A=@n26, B=LinearInterpolate(A=@n27, B=@n28)) {constB=0.02}) {materialFunction=&BreakOutFloat3Components}
  n30 = Add(A=Add(A=@n29.R, B=@n29.R), B=@n29.B)
  n31 = Multiply(
   A=Subtract(A=Time(), B=ScalarParameter() {group=Slurp, parameterName=Slurp_WPO_random_seed})
   B=ScalarParameter() {defaultValue=0.6, group=Slurp, parameterName=Slurp_WPO_LocalNoise_Speed}
  ) {constB=0.6}
  n32 = Sine(None=Subtract(A=Subtract(A=@n30) {constB=2}, B=Frac(None=@n31)))
  n33 = Multiply(A=Add(A=Add(A=@n32)))
  n34 = TextureSampleParameter2D() {
   parameterName=Slurp Gradient Mask
   samplerType=SAMPLERTYPE_Masks
   texture=/Game/Characters/Player/Male/Medium/Bodies/M_MED_Psyche/Textures/T_M_MED_Psyche_Liquid_M
  }
  n35 = Saturate(None=ChannelMaskParameter(None=@n34.RGB) {parameterName=Slurp Gradient Channel}.R)
  n36 = Multiply(
   A=@n33
   B=LinearInterpolate(
    A=ScalarParameter() {defaultValue=0.02, group=Slurp, parameterName=Slurp_WPO_LocalNoise_StrengthX}
    B=ScalarParameter() {defaultValue=0.02, group=Slurp, parameterName=Slurp_WPO_LocalNoise_StrengthX2}
    Alpha=@n35
   ) {constA=1, constB=0}
  ) {constB=0.2}
  n37 = Multiply(
   A=@n33
   B=LinearInterpolate(
    A=ScalarParameter() {defaultValue=0.02, group=Slurp, parameterName=Slurp_WPO_LocalNoise_StrengthY}
    B=ScalarParameter() {defaultValue=0.02, group=Slurp, parameterName=Slurp_WPO_LocalNoise_StrengthY2}
    Alpha=@n35
   ) {constA=1, constB=0}
  ) {constB=0.2}
  n38 = LinearInterpolate(
   A=ScalarParameter() {defaultValue=0.2, group=Slurp, parameterName=Slurp_WPO_LocalNoise_StrengthZ_Mult}
   B=ScalarParameter() {defaultValue=0.2, group=Slurp, parameterName=Slurp_WPO_LocalNoise_StrengthZ_Mult2}
   Alpha=@n35
  )
  n39 = ScalarParameter() {defaultValue=1, group=Slurp, parameterName=Slurp_WPO_LocalNoise_Speed_Far}
  n40 = ScalarParameter() {group=Slurp, parameterName=Slurp_WPO_LocalNoise_Speed_Near}
  n41 = Distance(A=ObjectPositionWS(), B=CameraPositionWS())
  n42 = ScalarParameter() {defaultValue=200, group=Slurp, parameterName=Slurp_WPO_LocalNoise_Speed_Distance}
  n43 = OneMinus(None=Saturate(None=Divide(A=@n41, B=@n42) {constB=100}))
  n44 = Multiply(
   A=AppendVector(A=AppendVector(A=@n36, B=@n37), B=Multiply(A=@n36, B=@n38) {constB=0.2})
   B=Multiply(A=@n35, B=LinearInterpolate(A=@n39, B=@n40, Alpha=@n43))
  )
  Multiply(A=VertexNormalWS(), B=@n44)
  n45 = VertexNormalWS()
  n46 = WorldPosition() {worldPositionShaderOffset=WPT_ExcludeAllShaderOffsets}
  n47 = Subtract(A=Frac(None=Multiply(A=@n46.XYZ) {constB=0.0025})) {constB=0.5}
  n48 = MaterialFunctionCall(Float3=Abs(None=Multiply(A=@n47) {constB=2})) {materialFunction=&BreakOutFloat3Components}
  n49 = Subtract(A=Add(A=Add(A=@n48.R, B=@n48.R), B=@n48.B)) {constB=2}
  n50 = Frac(None=Multiply(A=Time()) {constB=0.25})
  n51 = Add(A=Sine(None=Subtract(A=@n49, B=@n50)))
  n52 = Add(A=@n44, B=Multiply(A=Multiply(A=@n51) {constB=0.5}, B=ScalarParameter() {defaultValue=1, parameterName=WPO_WorldNoiseScale}))
  NamedRerouteDeclaration(None=LinearInterpolate(A=Multiply(A=@n45, B=@n52) {constB=0.5}, B=@n52) {constAlpha=0.8}) {name=WPO, nodeColor=<rgba 1 0 0.0703 1>}
  NamedRerouteDeclaration(None=@n17.Result) {name=Additional Normal, nodeColor=<rgba 0 1 0.3672 1>}
  n53 = SetMaterialAttributes!(
   MaterialAttributes=@n23.MaterialAttributes
   Base Color=LinearInterpolate(A=@n23.Base Color, B=VectorParameter() {parameterName=Secondary Color, defaultValue=<rgba 0.0022 0.0017 0.005 1>}.RGB, Alpha=@n15)
   Specular=ScalarParameter() {defaultValue=0.3238, parameterName=Specular}
   Roughness=ScalarParameter() {defaultValue=0.33, parameterName=Roughness}
   Emissive Color=@n20
   Normal=MaterialFunctionCall(BaseNormal=Constant3Vector() {constant=<rgba 0 0 1 0>}, AdditionalNormal=@n24) {materialFunction=&BlendAngleCorrectedNormals}.Result
   World Position Offset=NamedRerouteUsage()
   Metallic=Constant()
  ) {attributeSetTypes [
    #2
    #4
    #5
    #6
    #1
    #3
    #7
   ]}
  n54 = TextureSampleParameter2D() {parameterName=SkinFX_Mask, texture=&T_BlacksRGB}
  n55 = ChannelMaskParameter(None=@n54.RGB) {parameterName=Slurp Channel}
  n56 = MaterialFunctionCall(
   AO=@n21.AO
   Distance Blend=@n21.AO
   Fresnel=@n21.AO
   Gmap=@n21.AO
   Pre Skinned Local Position=@n21.AO
   Material=BlendMaterialAttributes(A=@n22.Result, B=@n53, Alpha=@n55.R) {vertexAttributeBlendType=UseA}
  ) {materialFunction=&BaseFN_CharacterFunction_PostFX}
  n57 = GetMaterialAttributes(None=@n56.Result) {attributeGetTypes [#3, #2]}
  n58 = StaticSwitchParameter(
   True=ChannelMaskParameter(None=@n54.RGB) {maskChannel=Blue, parameterName=Additional Wobbling Channel, defaultValue=<rgba 0 0 1 0>}.R
   False=@n55.R
  ) {parameterName=Additional Wobbling Channel}
  n59 = VertexColor()
  SetMaterialAttributes!(
   MaterialAttributes=@n57.MaterialAttributes
   World Position Offset=LinearInterpolate(A=@n57.World Position Offset, B=NamedRerouteUsage(), Alpha=LinearInterpolate(B=@n58, Alpha=@n59.G) {constB=0})
   Normal
  ) {attributeSetTypes [#3, #1]}
  VertexInterpolator(VS=@n26)
  WorldPosition()
 ]
}
